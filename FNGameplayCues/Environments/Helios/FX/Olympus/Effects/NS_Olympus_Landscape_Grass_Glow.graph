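# GPU Niagara system for glowing motes drifting over landscape grass in the Helios Olympus environment.
# Glows spawns around the system, snaps particles above sampled landscape, filters by GBuffer hue and culls those inside hard-coded world-space kill volumes.
# Time of day drives spawn rate and color, while wind, drag, waveform flicker and region-mask fading shape each sprite.
# Glows_Child001 spawns secondary sprites from Glows particles; the near-duplicate Glows_Child is disabled.
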NiagaraSystem /FNGameplayCues/Environments/Helios/FX/Olympus/Effects/NS_Olympus_Landscape_Grass_Glow
dependencies {
 usedDataInterfaces [
  &NiagaraDataInterfaceParticleRead
  &NiagaraDataInterfaceGBuffer
  &NiagaraDataInterfaceCurve
  &NiagaraDataInterfaceCollisionQuery
  &NiagaraDataInterfaceColorCurve
 ]
 usedDynamicInputs []
 usedModules [&SystemState]
}
emitters {
 Glows {
  simTarget=GPUComputeSim
  emitterUpdateScript [
   &EmitterState()
   &SpawnRate(
    SpawnRate=&Multiply_Float(
     A=300
     B=&FloatFromCurve(
      FloatCurve=DataInterface() {propertyValues {Curve {keys [<at 0 0.25>,<at 0.1 0.25>,<at 0.1667 1>,<at 0.3333 1>,<at 0.4 0.25>,<at 0.7 0.25>,<at 0.75 1>,<at 0.875 1>,<at 0.9 0.25>,<at 1 0.25>]}}}
      CurveIndex=Linked() {linkedVariable {name=NPC.FortniteNPC.TimeOfDay_0-1, type=NiagaraFloat}}
     )
    )
    Use Spawn Probability=false
   )
  ]
  particleSpawnScript [
   &InitializeParticle(
    Color=&FN_Select_Color_from_Time_Of_Day(Morning=<rgba 1 1 1 1>, Day=<rgba 250 1 1 1>, Evening=<rgba 1 1 1 1>, Night=<rgba 2 2 2 1>)
    Lifetime=&RandomRangeFloat(Minimum=1, Maximum=2)
    Mass=3
    Mass Mode=::NewEnumerator1
    Sprite Rotation Mode=::NewEnumerator2
    Sprite Size Mode=::NewEnumerator1
    Uniform Sprite Size Max=5
    Uniform Sprite Size Min=1
    UsePositionOffset=false
    Write Material Random=true
   ) {module=InitializeParticle001}
   &ShapeLocation(MaskSpawnGroup=false, Offset=<0 0 1>, Offset Mode=::NewEnumerator0, Override Seed=false, Sphere Radius=5000, UseOwnerScale=true)
   /SetVariables(Particles.Terrain Offset:NiagaraFloat=&RandomRangeFloat(Minimum=50, Maximum=100)) {isSetParameters=true}
   &self:Sample_Landscape_01() {module=Sample_Landscape_01}
   /SetVariables(Particles.Position=&Add_Vector(
     A=Linked() {linkedVariable {name=Particles.Position, type=NiagaraPosition}}
     B=&MakeVector(X=0, Y=0, Z=$Particles.Terrain Offset)
    )) {isSetParameters=true}
   &self:ScratchModule_08(Angle=25, BlendDirection=1, Normal:Vector3f=<0 0 1>, PickCeilingOrFloor=true, Velocity=2) {module=ScratchModule_08}
   /Game/Effects/Niagara/Modules/Spawn/FN_Spawn_GBuffer_Color_WithinRangeV2(HueMax=30, HueMin=20)
   &KillParticlesInVolume(
    Box Size:Vector3f=<10500.4678 10400.5605 5000.5693>
    Kill Shape=::NewEnumerator1
    Volume Origin:Vector3f=&AddVectorToPosition(Position=<-55068 65264 11316>, Vector=<0 0 0>)
   ) {module=KillParticlesInVolume001}
   &KillParticlesInVolume(
    Box Size:Vector3f=<4100.8931 2200.5056 2900.7883>
    Kill Shape=::NewEnumerator1
    Volume Origin:Vector3f=&AddVectorToPosition(Position=<-54960 59072 10028>, Vector=<0 0 0>)
   ) {module=KillParticlesInVolume002}
   &KillParticlesInVolume(
    Box Size:Vector3f=<2300.2544 2900.3179 5000.5693>
    Kill Shape=::NewEnumerator1
    Volume Origin:Vector3f=&AddVectorToPosition(Position=<-49416 64096 11316>, Vector=<0 0 0>)
   ) {module=KillParticlesInVolume003}
   &KillParticlesInVolume(
    Box Size:Vector3f=<7700.8633 5000.3853 5000.5693>
    Kill Shape=::NewEnumerator1
    Volume Origin:Vector3f=&AddVectorToPosition(Position=<-50116 67716 11316>, Vector=<0 0 0>)
   ) {module=KillParticlesInVolume004}
   &KillParticlesInVolume(
    Box Size:Vector3f=<2100.8604 2800.9553 4200.7671>
    Kill Shape=::NewEnumerator1
    Volume Origin:Vector3f=&AddVectorToPosition(Position=<-59656 71292 12952>, Vector=<0 0 0>)
   ) {module=KillParticlesInVolume005}
   &KillParticlesInVolume(
    Box Size:Vector3f=<6400.3945 2000.548 4200.7671>
    Kill Shape=::NewEnumerator1
    Volume Origin:Vector3f=&AddVectorToPosition(Position=<-62080 68196 10216>, Vector=<0 0 0>)
   ) {module=KillParticlesInVolume006}
   &KillParticlesInVolume(
    Box Size:Vector3f=<6400.3945 5000.1875 5000.5693>
    Kill Shape=::NewEnumerator1
    Volume Origin:Vector3f=&AddVectorToPosition(Position=<-65812 68196 11640>, Vector=<0 0 0>)
   ) {module=KillParticlesInVolume007}
   &KillParticlesInVolume(
    Box Size:Vector3f=<2100.8604 2800.9553 3000.8892>
    Kill Shape=::NewEnumerator1
    Volume Origin:Vector3f=&AddVectorToPosition(Position=<-54708 71292 11440>, Vector=<0 0 0>)
   ) {module=KillParticlesInVolume008}
   &KillParticlesInVolume(
    Box Size:Vector3f=<5800.4829 9000.3623 8700.3916>
    Kill Shape=::NewEnumerator1
    Volume Origin:Vector3f=&AddVectorToPosition(Position=<-54744 76168 13832>, Vector=<0 0 0>)
   ) {module=KillParticlesInVolume009}
   &KillParticlesInVolume(
    Sphere Radius=13700.3711
    Volume Origin:Vector3f=&AddVectorToPosition(Position=<-94628 59424 7464>, Vector=<0 0 0>)
   ) {module=KillParticlesInVolume010}
   &KillParticlesInVolume(
    Sphere Radius=4401
    Volume Origin:Vector3f=&AddVectorToPosition(Position=<-103096 46200 4268>, Vector=<0 0 0>)
   ) {module=KillParticlesInVolume011}
   &KillParticlesInVolume(
    Sphere Radius=2900.0178
    Volume Origin:Vector3f=&AddVectorToPosition(Position=<-80452 45000 896>, Vector=<0 0 0>)
   ) {module=KillParticlesInVolume012}
   &KillParticlesInVolume(
    Box Size:Vector3f=<5900.3345 9400.8809 5900.3345>
    Kill Shape=::NewEnumerator1
    Volume Origin:Vector3f=&AddVectorToPosition(Position=<-51236 32960 7956>, Vector=<0 0 0>)
   ) {module=KillParticlesInVolume013}
   &KillParticlesInVolume(
    Sphere Radius=10200.9619
    Volume Origin:Vector3f=&AddVectorToPosition(Position=<-27684 94512 12464>, Vector=<0 0 0>)
   ) {module=KillParticlesInVolume014}
   &KillParticlesInVolume(
    Sphere Radius=2300.4858
    Volume Origin:Vector3f=&AddVectorToPosition(Position=<-31152 38824 7576>, Vector=<0 0 0>)
   ) {module=KillParticlesInVolume015}
  ]
  particleUpdateScript [
   /SetVariables() {isSetParameters=true}
   &ParticleState()
   &RegionMask(Region Radius=500, UseOffset=false, UseRotation=false)
   /SetVariables(Particles.Normalized Velocity:Vector3f=&NormalizeVector(Vector To Normalize=$Particles.Velocity)) {isSetParameters=true, module=SetVariables_#1}
   &SolveForcesAndVelocity(Clamp Velocity=true, Limit Acceleration=false, Speed Limit=100)
   &ScaleSpriteSize(
    Scale Sprite Size Mode=::NewEnumerator0
    Uniform Scale Factor=&FloatFromCurve(
     FloatCurve=DataInterface() {propertyValues {Curve {keys [
         {interpMode=RCIM_Cubic, tangentMode=RCTM_User, time=0, value=0}
         {interpMode=RCIM_Cubic, tangentMode=RCTM_User, time=0.5, value=1}
         {tangentMode=RCTM_User, time=1, value=0}
        ]}}}
     Scale Curve=100
    )
   )
   &WindForce(
    Frequency=5
    Friction=0
    Offset By Time:Vector3f=<0 0 -10>
    Scale Min / Max:Vector2f=<1 0.5>
    Speed Min / Max:Vector2f=<0 100>
    Turbulence Scale Mode:ENiagaraWindTurbulenceContributionMode=::NewEnumerator1
    Use Ground Mask=true
    Wind Speed=<5 0 5>
   ) {module=WindForce001}
   &AerodynamicDrag(
    Aerodynamic Rotational Drag=0.1
    Drag Shape Curve=DataInterface() {propertyValues {Curve {keys [
        {interpMode=RCIM_Cubic, tangentMode=RCTM_User, time=0, value=0.1}
        {arriveTangent=2.0799, interpMode=RCIM_Cubic, leaveTangent=2.0799, tangentMode=RCTM_User, time=1, value=1}
       ]}}}
    Lift Shape Curve=DataInterface() {propertyValues {Curve {keys [
        {interpMode=RCIM_Cubic, tangentMode=RCTM_User, time=0, value=0}
        {interpMode=RCIM_Cubic, tangentMode=RCTM_User, time=0.5, value=1}
        {tangentMode=RCTM_User, time=1, value=0}
       ]}}}
    Pivot Offset Maximum:Vector3f=<1 1 1>
    Pivot Offset Minimum Length:Vector3f=<0 0 0>
    Rotation Strength:NiagaraFloat=&RandomRangeFloat(Minimum=15, Maximum=25)
    Use Linear Drag=true
    Use Rotational Drag=true
   )
   &SolveForcesAndVelocity(Clamp Velocity=true, Limit Acceleration=false, Speed Limit=100) {module=SolveForcesAndVelocity001}
   &DynamicMaterialParameters(
    Index 0 Param 1=&RandomRangeFloat(Minimum=0.25, Maximum=0.5)
    Index 0 Param 2=&RandomRangeFloat(Minimum=0, Maximum=1)
    Index 0 Param 3=1
    Index 0 Param 4=0.1
    Index 1 Param 1=&RandomRangeFloat(Minimum=0.1, Maximum=0.35)
    Index 1 Param 2=&RandomRangeFloat(Minimum=0, Maximum=1)
    Index 1 Param 3=2
    Index 1 Param 4=1
    Index 2 Param 1=4096
    Index 2 Param 2=1
    Module.Param0WriteEnabled=true
    Module.Param0WriteEnabled1=true
    Module.Param0WriteEnabled2=true
    Module.Param1WriteEnabled=true
    Module.Param1WriteEnabled1=true
    Module.Param1WriteEnabled2=true
    Module.Param2WriteEnabled=true
    Module.Param2WriteEnabled1=true
    Module.Param2WriteEnabled2=true
    Module.Param3WriteEnabled=true
    Module.Param3WriteEnabled1=true
    Module.Param3WriteEnabled2=true
    Write Parameter Index 1=true
    Write Parameter Index 2=true
   )
   &ScaleColor(
    Scale Alpha=&Lerp_Float(
     A=&FloatFromCurve(FloatCurve=DataInterface() {propertyValues {Curve {keys [
          {interpMode=RCIM_Cubic, tangentMode=RCTM_User, time=0, value=0}
          {interpMode=RCIM_Cubic, tangentMode=RCTM_User, time=0.5, value=1}
          {tangentMode=RCTM_User, time=1, value=0}
         ]}}})
     B=0
     Alpha=$Particles.RegionMask.Float
    )
    Scale RGB=&Multiply_Vector(
     A=&MakeVectorFromLinearColorRGB(Color=&HueShiftLinearColor(
       Hue=&RandomRangeFloat(Minimum=1, Maximum=1.05)
       Saturation=&RandomRangeFloat(Minimum=1, Maximum=2)
       Value=&Multiply_Float(
        A=&FloatFromCurve(FloatCurve=DataInterface() {propertyValues {Curve {keys [
             {interpMode=RCIM_Cubic, tangentMode=RCTM_User, time=0, value=0}
             {interpMode=RCIM_Cubic, tangentMode=RCTM_User, time=0.5, value=1}
             {tangentMode=RCTM_User, time=1, value=0}
            ]}}})
        B=&Waveform(
         "Waveform [1]"=::NewEnumerator0
         "[1] Phase Scale"=&RandomRangeFloat(Minimum=0.25, Maximum=0.5)
         "[1] Frequency"=&RandomRangeFloat(Minimum=0.5, Maximum=1)
         "[1] Amplitude Min/Max"=<0 1>
         "[1] Scale"=&RandomRangeFloat(Minimum=1, Maximum=5)
        )
       )
       Linear Color=&ColorFromCurve(
        ColorCurve=DataInterface() {propertyValues {
          AlphaCurve {keys [<at 0.0013 1>,<at 1 1>]}
          BlueCurve {keys [<at 0 0>,<at 0 0>,<at 0.2584 0>,<at 0.5 0.1779>,<at 1 0>]}
          GreenCurve {keys [<at 0 0.1896>,<at 0 0.1896>,<at 0.2584 0.6184>,<at 0.5 0.8148>,<at 1 0.1896>]}
          RedCurve {keys [<at 0 0.7604>,<at 0 0.7604>,<at 0.2584 1>,<at 0.5 1>,<at 1 0.7604>]}
         }}
        CurveIndex=Linked() {linkedVariable {name=NPC.FortniteNPC.TimeOfDay_0-1, type=NiagaraFloat}}
       )
      ))
     B=&VectorFromFloat(Value=1)
    )
    ScaleA=true
    ScaleRGB=true
   ) {enabled=false}
   &ScaleColor(
    Scale Alpha=&Lerp_Float(
     A=&FloatFromCurve(FloatCurve=DataInterface() {propertyValues {Curve {keys [
          {interpMode=RCIM_Cubic, tangentMode=RCTM_User, time=0, value=0}
          {interpMode=RCIM_Cubic, tangentMode=RCTM_User, time=0.5, value=1}
          {tangentMode=RCTM_User, time=1, value=0}
         ]}}})
     B=0
     Alpha=$Particles.RegionMask.Float
    )
    Scale RGB=&Multiply_Vector(
     A=&MakeVectorFromLinearColorRGB(Color=&HueShiftLinearColor(
       Hue=&RandomRangeFloat(Minimum=1, Maximum=1.05)
       Saturation=&RandomRangeFloat(Minimum=1, Maximum=2)
       Value=&Multiply_Float(
        A=&FloatFromCurve(FloatCurve=DataInterface() {propertyValues {Curve {keys [
             {interpMode=RCIM_Cubic, tangentMode=RCTM_User, time=0, value=0}
             {interpMode=RCIM_Cubic, tangentMode=RCTM_User, time=0.5, value=1}
             {tangentMode=RCTM_User, time=1, value=0}
            ]}}})
        B=&Waveform(
         "Waveform [1]"=::NewEnumerator0
         "[1] Phase Scale"=&RandomRangeFloat(Minimum=0.25, Maximum=0.5)
         "[1] Frequency"=&RandomRangeFloat(Minimum=0.5, Maximum=1)
         "[1] Amplitude Min/Max"=<0 1>
         "[1] Scale"=&RandomRangeFloat(Minimum=1, Maximum=5)
        )
       )
       Linear Color=&ColorFromCurve(
        ColorCurve=DataInterface() {propertyValues {
          AlphaCurve {keys [<at 0.0013 1>,<at 1 1>]}
          BlueCurve {keys [<at 0 0.0348>,<at 0.2584 0>,<at 0.5 0.1779>,<at 1 0.0134>]}
          GreenCurve {keys [<at 0 0.0595>,<at 0.2584 0.6184>,<at 0.5 0.8148>,<at 1 0.1287>]}
          RedCurve {keys [<at 0 0.099>,<at 0.2584 1>,<at 0.5 1>,<at 1 0.3337>]}
         }}
        CurveIndex=Linked() {linkedVariable {name=NPC.FortniteNPC.TimeOfDay_0-1, type=NiagaraFloat}}
       )
      ))
     B=&VectorFromFloat(Value=1)
    )
    ScaleA=true
    ScaleRGB=true
   ) {module=ScaleColor001}
   /Niagara/Modules/Update/Camera/MaintainInCameraParticleScale() {module=MaintainInCameraParticleScale001}
   &CameraOffset(Camera Offset Amount=25)
  ]
  rendererClasses [&NiagaraSpriteRendererProperties]
  renderers [{rendererClass=&NiagaraSpriteRendererProperties}]
 }
 Glows_Child {
  bEnabled=false
  simTarget=GPUComputeSim
  emitterUpdateScript [
   &EmitterState()
   &SpawnParticlesFromOtherEmitter(
    Attribute Reader=DataInterface() {propertyValues {EmitterBinding {bindingMode=Other, emitterName=Glows}}}
    Use Maximum Age=false
    Use Minimum Age=false
   )
  ]
  particleSpawnScript [
   &InitializeParticle(
    Color Minimum=<rgba 1 1 1 1>
    Color Mode=::NewEnumerator2
    Lifetime=&RandomRangeFloat(Minimum=1, Maximum=2)
    Mass=3
    Mass Mode=::NewEnumerator1
    Sprite Rotation Mode=::NewEnumerator2
    Sprite Size Mode=::NewEnumerator1
    Uniform Sprite Size Max=3
    Uniform Sprite Size Min=1
    UsePositionOffset=false
    Write Material Random=true
   ) {module=InitializeParticle001}
   &ShapeLocation(MaskSpawnGroup=false, Offset=<0 0 1>, Offset Mode=::NewEnumerator0, Override Seed=false, Sphere Radius=50, UseOwnerScale=true)
   &SampleParticlesFromOtherEmitter()
   &FN_Spawn_GBuffer_Color_WithinRangeV2(HueMax=30, HueMin=20) {enabled=false, module=FN_Spawn_GBuffer_Color_WithinRangeV2001}
  ]
  particleUpdateScript [
   &ParticleState()
   &RegionMask(Region Radius=500, UseOffset=false, UseRotation=false)
   /SetVariables(Particles.Normalized Velocity:Vector3f=&NormalizeVector(Vector To Normalize=$Particles.Velocity)) {isSetParameters=true, module=SetVariables_#1}
   &CurlNoiseForce(Noise Frequency=25, Noise Strength=5, Pan Noise=false) {enabled=false, module=CurlNoiseForce001}
   &GravityForce(Gravity=&MakeVector(
     X=0
     Y=0
     Z=&FloatFromCurve(
      FloatCurve=DataInterface() {propertyValues {Curve {keys [
          <at 0 0>
          {interpMode=RCIM_Cubic, tangentMode=RCTM_User, time=0.02, value=2}
          <at 0.1 0>
          {interpMode=RCIM_Cubic, tangentMode=RCTM_User, time=0.2, value=-0.9}
          <at 0.5 -0.25>
          <at 0.8 0.1>
          <at 1 0.1>
         ]}}}
      Scale Curve=250
     )
    )) {enabled=false, module=GravityForce001}
   &SolveForcesAndVelocity(Clamp Velocity=true, Limit Acceleration=false, Speed Limit=100)
   &ScaleSpriteSize(
    Scale Sprite Size Mode=::NewEnumerator0
    Uniform Scale Factor=&FloatFromCurve(
     FloatCurve=DataInterface() {propertyValues {Curve {keys [
         {interpMode=RCIM_Cubic, tangentMode=RCTM_User, time=0, value=0}
         {interpMode=RCIM_Cubic, tangentMode=RCTM_User, time=0.5, value=1}
         {tangentMode=RCTM_User, time=1, value=0}
        ]}}}
     Scale Curve=5
    )
   )
   &WindForce(
    Frequency=5
    Friction=0
    Offset By Time:Vector3f=<0 0 -10>
    Scale Min / Max:Vector2f=<1 0.5>
    Speed Min / Max:Vector2f=<0 100>
    Turbulence Scale Mode:ENiagaraWindTurbulenceContributionMode=::NewEnumerator1
    Use Ground Mask=true
    Wind Speed=<5 0 5>
   ) {module=WindForce001}
   &AerodynamicDrag(
    Aerodynamic Rotational Drag=0.1
    Drag Shape Curve=DataInterface() {propertyValues {Curve {keys [
        {interpMode=RCIM_Cubic, tangentMode=RCTM_User, time=0, value=0.1}
        {arriveTangent=2.0799, interpMode=RCIM_Cubic, leaveTangent=2.0799, tangentMode=RCTM_User, time=1, value=1}
       ]}}}
    Lift Shape Curve=DataInterface() {propertyValues {Curve {keys [
        {interpMode=RCIM_Cubic, tangentMode=RCTM_User, time=0, value=0}
        {interpMode=RCIM_Cubic, tangentMode=RCTM_User, time=0.5, value=1}
        {tangentMode=RCTM_User, time=1, value=0}
       ]}}}
    Pivot Offset Maximum:Vector3f=<1 1 1>
    Pivot Offset Minimum Length:Vector3f=<0 0 0>
    Rotation Strength:NiagaraFloat=&RandomRangeFloat(Minimum=15, Maximum=25)
    Use Linear Drag=true
    Use Rotational Drag=true
   )
   &SolveForcesAndVelocity(Clamp Velocity=true, Limit Acceleration=false, Speed Limit=100) {module=SolveForcesAndVelocity001}
   &DynamicMaterialParameters(
    Index 0 Param 1=&RandomRangeFloat(Minimum=0.25, Maximum=0.5)
    Index 0 Param 2=&RandomRangeFloat(Minimum=0, Maximum=1)
    Index 0 Param 3=1
    Index 0 Param 4=0.1
    Index 1 Param 1=&RandomRangeFloat(Minimum=0.1, Maximum=0.5)
    Index 1 Param 2=&RandomRangeFloat(Minimum=0, Maximum=1)
    Index 1 Param 3=2
    Index 1 Param 4=1
    Index 2 Param 1=4096
    Index 2 Param 2=1
    Module.Param0WriteEnabled=true
    Module.Param0WriteEnabled1=true
    Module.Param0WriteEnabled2=true
    Module.Param1WriteEnabled=true
    Module.Param1WriteEnabled1=true
    Module.Param1WriteEnabled2=true
    Module.Param2WriteEnabled=true
    Module.Param2WriteEnabled1=true
    Module.Param2WriteEnabled2=true
    Module.Param3WriteEnabled=true
    Module.Param3WriteEnabled1=true
    Module.Param3WriteEnabled2=true
    Write Parameter Index 1=true
    Write Parameter Index 2=true
   )
   &ScaleColor(
    Scale Alpha=&Lerp_Float(
     A=&FloatFromCurve(
      FloatCurve=DataInterface() {propertyValues {Curve {keys [
          {interpMode=RCIM_Cubic, tangentMode=RCTM_User, time=0, value=0}
          {interpMode=RCIM_Cubic, tangentMode=RCTM_User, time=0.5, value=1}
          {tangentMode=RCTM_User, time=1, value=0}
         ]}}}
      Scale Curve=0.02
     )
     B=0
     Alpha=$Particles.RegionMask.Float
    )
    Scale RGB=&Multiply_Vector(
     A=&MakeVectorFromLinearColorRGB(Color=&HueShiftLinearColor(
       Hue=&RandomRangeFloat(Minimum=0.8, Maximum=1.07)
       Saturation=&RandomRangeFloat(Minimum=0.25, Maximum=1)
       Value=&Multiply_Float(
        A=&FloatFromCurve(FloatCurve=DataInterface() {propertyValues {Curve {keys [
             {interpMode=RCIM_Cubic, tangentMode=RCTM_User, time=0, value=0}
             {interpMode=RCIM_Cubic, tangentMode=RCTM_User, time=0.5, value=1}
             {tangentMode=RCTM_User, time=1, value=0}
            ]}}})
        B=&Waveform(
         "Waveform [1]"=::NewEnumerator0
         "[1] Phase Scale"=&RandomRangeFloat(Minimum=0.25, Maximum=0.5)
         "[1] Frequency"=&RandomRangeFloat(Minimum=0.5, Maximum=1)
         "[1] Amplitude Min/Max"=<0 1>
         "[1] Scale"=&RandomRangeFloat(Minimum=1, Maximum=5)
        )
       )
       Linear Color=&ColorFromCurve(
        ColorCurve=DataInterface() {propertyValues {
          AlphaCurve {keys [<at 0.0013 1>,<at 1 1>]}
          BlueCurve {keys [<at 0 0>,<at 0 0>,<at 0.2584 0>,<at 0.5 0.1779>,<at 1 0>]}
          GreenCurve {keys [<at 0 0.1896>,<at 0 0.1896>,<at 0.2584 0.6184>,<at 0.5 0.8148>,<at 1 0.1896>]}
          RedCurve {keys [<at 0 0.7604>,<at 0 0.7604>,<at 0.2584 1>,<at 0.5 1>,<at 1 0.7604>]}
         }}
        CurveIndex=Linked() {linkedVariable {name=NPC.FortniteNPC.TimeOfDay_0-1, type=NiagaraFloat}}
       )
      ))
     B=&VectorFromFloat(Value=1)
    )
    ScaleA=true
    ScaleRGB=true
   )
   &MaintainInCameraParticleScale(Pixel Depth=2) {module=MaintainInCameraParticleScale001}
   &CameraOffset(Camera Offset Amount=25)
  ]
  rendererClasses [&NiagaraSpriteRendererProperties]
  renderers [
   {rendererClass=&NiagaraSpriteRendererProperties}
   {rendererClass=&NiagaraSpriteRendererProperties, rendererIndex=1}
  ]
 }
 Glows_Child001 {
  simTarget=GPUComputeSim
  emitterUpdateScript [
   &EmitterState()
   &SpawnParticlesFromOtherEmitter(
    Attribute Reader=DataInterface() {propertyValues {EmitterBinding {bindingMode=Other, emitterName=Glows}}}
    Use Maximum Age=false
    Use Minimum Age=false
   )
  ]
  particleSpawnScript [
   &InitializeParticle(
    Color Minimum=<rgba 1 1 1 1>
    Color Mode=::NewEnumerator2
    Lifetime=&RandomRangeFloat(Minimum=1, Maximum=2)
    Mass=3
    Mass Mode=::NewEnumerator1
    Sprite Rotation Mode=::NewEnumerator2
    Sprite Size Mode=::NewEnumerator1
    Uniform Sprite Size Max=1
    Uniform Sprite Size Min=1
    UsePositionOffset=false
    Write Material Random=true
   ) {module=InitializeParticle001}
   &ShapeLocation(MaskSpawnGroup=false, Offset=<0 0 1>, Offset Mode=::NewEnumerator0, Override Seed=false, Sphere Radius=50, UseOwnerScale=true)
   &SampleParticlesFromOtherEmitter()
   &FN_Spawn_GBuffer_Color_WithinRangeV2(HueMax=30, HueMin=20) {enabled=false, module=FN_Spawn_GBuffer_Color_WithinRangeV2001}
  ]
  particleUpdateScript [
   &ParticleState()
   &RegionMask(Region Radius=500, UseOffset=false, UseRotation=false)
   /SetVariables(Particles.Normalized Velocity:Vector3f=&NormalizeVector(Vector To Normalize=$Particles.Velocity)) {isSetParameters=true, module=SetVariables_#1}
   &CurlNoiseForce(Noise Frequency=25, Noise Strength=5, Pan Noise=false) {enabled=false, module=CurlNoiseForce001}
   &GravityForce(Gravity=&MakeVector(
     X=0
     Y=0
     Z=&FloatFromCurve(
      FloatCurve=DataInterface() {propertyValues {Curve {keys [
          <at 0 0>
          {interpMode=RCIM_Cubic, tangentMode=RCTM_User, time=0.02, value=2}
          <at 0.1 0>
          {interpMode=RCIM_Cubic, tangentMode=RCTM_User, time=0.2, value=-0.9}
          <at 0.5 -0.25>
          <at 0.8 0.1>
          <at 1 0.1>
         ]}}}
      Scale Curve=250
     )
    )) {enabled=false, module=GravityForce001}
   &SolveForcesAndVelocity(Clamp Velocity=true, Limit Acceleration=false, Speed Limit=100)
   &ScaleSpriteSize(
    Scale Sprite Size Mode=::NewEnumerator0
    Uniform Scale Factor=&FloatFromCurve(
     FloatCurve=DataInterface() {propertyValues {Curve {keys [
         {interpMode=RCIM_Cubic, tangentMode=RCTM_User, time=0, value=0}
         {interpMode=RCIM_Cubic, tangentMode=RCTM_User, time=0.5, value=1}
         {tangentMode=RCTM_User, time=1, value=0}
        ]}}}
     Scale Curve=5
    )
   )
   &WindForce(
    Frequency=5
    Friction=0
    Offset By Time:Vector3f=<0 0 -10>
    Scale Min / Max:Vector2f=<1 0.5>
    Speed Min / Max:Vector2f=<0 100>
    Turbulence Scale Mode:ENiagaraWindTurbulenceContributionMode=::NewEnumerator1
    Use Ground Mask=true
    Wind Speed=<5 0 5>
   ) {module=WindForce001}
   &AerodynamicDrag(
    Aerodynamic Rotational Drag=0.1
    Drag Shape Curve=DataInterface() {propertyValues {Curve {keys [
        {interpMode=RCIM_Cubic, tangentMode=RCTM_User, time=0, value=0.1}
        {arriveTangent=2.0799, interpMode=RCIM_Cubic, leaveTangent=2.0799, tangentMode=RCTM_User, time=1, value=1}
       ]}}}
    Lift Shape Curve=DataInterface() {propertyValues {Curve {keys [
        {interpMode=RCIM_Cubic, tangentMode=RCTM_User, time=0, value=0}
        {interpMode=RCIM_Cubic, tangentMode=RCTM_User, time=0.5, value=1}
        {tangentMode=RCTM_User, time=1, value=0}
       ]}}}
    Pivot Offset Maximum:Vector3f=<1 1 1>
    Pivot Offset Minimum Length:Vector3f=<0 0 0>
    Rotation Strength:NiagaraFloat=&RandomRangeFloat(Minimum=15, Maximum=25)
    Use Linear Drag=true
    Use Rotational Drag=true
   )
   &SolveForcesAndVelocity(Clamp Velocity=true, Limit Acceleration=false, Speed Limit=100) {module=SolveForcesAndVelocity001}
   &DynamicMaterialParameters(
    Index 0 Param 1=&RandomRangeFloat(Minimum=0.25, Maximum=0.5)
    Index 0 Param 2=&RandomRangeFloat(Minimum=0, Maximum=1)
    Index 0 Param 3=1
    Index 0 Param 4=0.1
    Index 1 Param 1=&RandomRangeFloat(Minimum=0.1, Maximum=0.5)
    Index 1 Param 2=&RandomRangeFloat(Minimum=0, Maximum=1)
    Index 1 Param 3=2
    Index 1 Param 4=1
    Index 2 Param 1=4096
    Index 2 Param 2=1
    Module.Param0WriteEnabled=true
    Module.Param0WriteEnabled1=true
    Module.Param0WriteEnabled2=true
    Module.Param1WriteEnabled=true
    Module.Param1WriteEnabled1=true
    Module.Param1WriteEnabled2=true
    Module.Param2WriteEnabled=true
    Module.Param2WriteEnabled1=true
    Module.Param2WriteEnabled2=true
    Module.Param3WriteEnabled=true
    Module.Param3WriteEnabled1=true
    Module.Param3WriteEnabled2=true
    Write Parameter Index 1=true
    Write Parameter Index 2=true
   )
   &ScaleColor(
    Scale Alpha=&Lerp_Float(
     A=&FloatFromCurve(
      FloatCurve=DataInterface() {propertyValues {Curve {keys [
          {interpMode=RCIM_Cubic, tangentMode=RCTM_User, time=0, value=0}
          {interpMode=RCIM_Cubic, tangentMode=RCTM_User, time=0.5, value=1}
          {tangentMode=RCTM_User, time=1, value=0}
         ]}}}
      Scale Curve=0.02
     )
     B=0
     Alpha=$Particles.RegionMask.Float
    )
    Scale RGB=&Multiply_Vector(
     A=&MakeVectorFromLinearColorRGB(Color=&HueShiftLinearColor(
       Hue=&RandomRangeFloat(Minimum=0.8, Maximum=1.07)
       Saturation=&RandomRangeFloat(Minimum=0.25, Maximum=1)
       Value=&Multiply_Float(
        A=&FloatFromCurve(FloatCurve=DataInterface() {propertyValues {Curve {keys [
             {interpMode=RCIM_Cubic, tangentMode=RCTM_User, time=0, value=0}
             {interpMode=RCIM_Cubic, tangentMode=RCTM_User, time=0.5, value=1}
             {tangentMode=RCTM_User, time=1, value=0}
            ]}}})
        B=&Waveform(
         "Waveform [1]"=::NewEnumerator0
         "[1] Phase Scale"=&RandomRangeFloat(Minimum=0.25, Maximum=0.5)
         "[1] Frequency"=&RandomRangeFloat(Minimum=0.5, Maximum=1)
         "[1] Amplitude Min/Max"=<0 1>
         "[1] Scale"=&RandomRangeFloat(Minimum=1, Maximum=5)
        )
       )
       Linear Color=&ColorFromCurve(
        ColorCurve=DataInterface() {propertyValues {
          AlphaCurve {keys [<at 0.0013 1>,<at 1 1>]}
          BlueCurve {keys [<at 0 0>,<at 0 0>,<at 0.2584 0>,<at 0.5 0.1779>,<at 1 0>]}
          GreenCurve {keys [<at 0 0.1896>,<at 0 0.1896>,<at 0.2584 0.6184>,<at 0.5 0.8148>,<at 1 0.1896>]}
          RedCurve {keys [<at 0 0.7604>,<at 0 0.7604>,<at 0.2584 1>,<at 0.5 1>,<at 1 0.7604>]}
         }}
        CurveIndex=Linked() {linkedVariable {name=NPC.FortniteNPC.TimeOfDay_0-1, type=NiagaraFloat}}
       )
      ))
     B=&VectorFromFloat(Value=1)
    )
    ScaleA=true
    ScaleRGB=true
   )
   &MaintainInCameraParticleScale(Pixel Depth=2) {module=MaintainInCameraParticleScale001}
   &CameraOffset(Camera Offset Amount=25)
  ]
  rendererClasses [&NiagaraSpriteRendererProperties]
  renderers [
   {rendererClass=&NiagaraSpriteRendererProperties}
   {rendererClass=&NiagaraSpriteRendererProperties, rendererIndex=1}
  ]
 }
}
summary {
 systemName=NS_Olympus_Landscape_Grass_Glow
 emitters [
  {bEnabled=false, emitterName=Glows_Child, simTarget=GPUComputeSim, rendererClasses [&NiagaraSpriteRendererProperties]}
  {emitterName=Glows, simTarget=GPUComputeSim, rendererClasses [&NiagaraSpriteRendererProperties]}
  {emitterName=Glows_Child001, simTarget=GPUComputeSim, rendererClasses [&NiagaraSpriteRendererProperties]}
 ]
 userVariables [
  User.Alpha_AttractionStrength {type=NiagaraFloat, defaultValue=0, description=""}
  User.Landscape {
   type=NiagaraDataInterfaceLandscape
   description=""
   defaultValue {
    struct=NiagaraExt_VariableValue_DataInterface
    value {
     dataInterface=&self:NiagaraDataInterfaceLandscape_0
     dataInterfaceClass=&NiagaraDataInterfaceLandscape
    }
   }
  }
  User.Player Position {type=Vector3f, description="", defaultValue {struct=Vector3f, value=<0 0 0>}}
 ]
}
